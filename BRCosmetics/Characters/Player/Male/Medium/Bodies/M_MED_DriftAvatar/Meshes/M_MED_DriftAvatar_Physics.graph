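# Physics asset for the male medium DriftAvatar cosmetic body, covering the dangling vest bones and the spine.
# Seven vest chains (center back, plus left and right back, side and front) each run from a tiny master capsule through two larger joint capsules.
# Chain joints have tight swing and tighter twist limits, while the masters attach to spine_03 with free rotation.
# Three wide spine capsules, also joined with free rotation, form the torso volume.
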PhysicsAsset /BRCosmetics/Characters/Player/Male/Medium/Bodies/M_MED_DriftAvatar/Meshes/M_MED_DriftAvatar_Physics
bodies {
 dyn_c_BackVest_01_MSTR [{shapeType=Capsule, length=0.505, radius=0.505, shapeName=dyn_c_BackVest_01_MSTR_capsule, rotation=<rot 0 -90 0>}]
 dyn_c_BackVest_01_jnt [{shapeType=Capsule, length=5, radius=4, shapeName=dyn_c_BackVest_01_jnt_capsule, center=<5 0 1>, rotation=<rot 0 0 90>}]
 dyn_c_BackVest_02_jnt [{shapeType=Capsule, length=5, radius=4, shapeName=dyn_c_BackVest_02_jnt_capsule, center=<5 0 1>, rotation=<rot 0 0 90>}]
 dyn_l_BackVest_01_MSTR [{shapeType=Capsule, length=0.505, radius=0.505, shapeName=dyn_l_BackVest_01_MSTR_capsule, rotation=<rot 0 -90 0>}]
 dyn_l_BackVest_01_jnt [{shapeType=Capsule, length=5, radius=4, shapeName=dyn_l_BackVest_01_jnt_capsule, center=<5 0 1>, rotation=<rot 0 0 90>}]
 dyn_l_BackVest_02_jnt [{shapeType=Capsule, length=5, radius=4, shapeName=dyn_l_BackVest_02_jnt_capsule, center=<5 0 1>, rotation=<rot 0 0 90>}]
 dyn_l_FrontVest_01_MSTR [{shapeType=Capsule, length=0.505, radius=0.505, shapeName=dyn_l_FrontVest_01_MSTR_capsule, rotation=<rot 0 -90 0>}]
 dyn_l_FrontVest_01_jnt [{shapeType=Capsule, length=5, radius=4, shapeName=dyn_l_FrontVest_01_jnt_capsule, center=<5 0 1>, rotation=<rot 0 0 90>}]
 dyn_l_FrontVest_02_jnt [{shapeType=Capsule, length=5, radius=4, shapeName=dyn_l_FrontVest_02_jnt_capsule, center=<5 0 1>, rotation=<rot 0 0 90>}]
 dyn_l_SideVest_01_MSTR [{shapeType=Capsule, length=0.505, radius=0.505, shapeName=dyn_l_SideVest_01_MSTR_capsule, rotation=<rot 0 -90 0>}]
 dyn_l_SideVest_01_jnt [{shapeType=Capsule, length=5, radius=4, shapeName=dyn_l_SideVest_01_jnt_capsule, center=<5 0 1>, rotation=<rot 0 0 90>}]
 dyn_l_SideVest_02_jnt [{shapeType=Capsule, length=5, radius=4, shapeName=dyn_l_SideVest_02_jnt_capsule, center=<5 0 1>, rotation=<rot 0 0 90>}]
 dyn_r_BackVest_01_MSTR [{shapeType=Capsule, length=0.505, radius=0.505, shapeName=dyn_r_BackVest_01_MSTR_capsule, rotation=<rot 0 -90 0>}]
 dyn_r_BackVest_01_jnt [{shapeType=Capsule, length=5, radius=4, shapeName=dyn_r_BackVest_01_jnt_capsule, center=<5 0 1>, rotation=<rot 0 0 90>}]
 dyn_r_BackVest_02_jnt [{shapeType=Capsule, length=5, radius=4, shapeName=dyn_r_BackVest_02_jnt_capsule, center=<5 0 1>, rotation=<rot 0 0 90>}]
 dyn_r_FrontVest_01_MSTR [{shapeType=Capsule, length=0.505, radius=0.505, shapeName=dyn_r_FrontVest_01_MSTR_capsule, rotation=<rot 0 -90 0>}]
 dyn_r_FrontVest_01_jnt [{shapeType=Capsule, length=5, radius=4, shapeName=dyn_r_FrontVest_01_jnt_capsule, center=<5 0 1>, rotation=<rot 0 0 90>}]
 dyn_r_FrontVest_02_jnt [{shapeType=Capsule, length=5, radius=4, shapeName=dyn_r_FrontVest_02_jnt_capsule, center=<5 0 1>, rotation=<rot 0 0 90>}]
 dyn_r_SideVest_01_MSTR [{shapeType=Capsule, length=0.505, radius=0.505, shapeName=dyn_r_SideVest_01_MSTR_capsule, rotation=<rot 0 -90 0>}]
 dyn_r_SideVest_01_jnt [{shapeType=Capsule, length=5, radius=4, shapeName=dyn_r_SideVest_01_jnt_capsule, center=<5 0 1>, rotation=<rot 0 0 90>}]
 dyn_r_SideVest_02_jnt [{shapeType=Capsule, length=5, radius=4, shapeName=dyn_r_SideVest_02_jnt_capsule, center=<5 0 1>, rotation=<rot 0 0 90>}]
 spine_01 [{shapeType=Capsule, length=2, radius=14.15, shapeName=spine_01_capsule, rotation=<rot 90 0 0>}]
 spine_02 [{shapeType=Capsule, length=2, radius=14.15, shapeName=spine_02_capsule, center=<2.7736 -0.0721 -0.0872>, rotation=<rot 90 0 0>}]
 spine_03 [{shapeType=Capsule, length=2, radius=14.15, shapeName=spine_03_capsule, center=<6.6402 -1.2134 0.0163>, rotation=<rot 90 0 0>}]
}
constraints [
 {bone1Name=dyn_c_BackVest_01_jnt, bone2Name=dyn_c_BackVest_01_MSTR, swing1LimitDegrees=15, swing2LimitDegrees=15, twistLimitDegrees=2}
 {bone1Name=dyn_c_BackVest_02_jnt, bone2Name=dyn_c_BackVest_01_jnt, swing1LimitDegrees=15, swing2LimitDegrees=15, twistLimitDegrees=2}
 {bone1Name=dyn_l_BackVest_01_jnt, bone2Name=dyn_l_BackVest_01_MSTR, swing1LimitDegrees=15, swing2LimitDegrees=15, twistLimitDegrees=2}
 {bone1Name=dyn_l_BackVest_02_jnt, bone2Name=dyn_l_BackVest_01_jnt, swing1LimitDegrees=15, swing2LimitDegrees=15, twistLimitDegrees=2}
 {bone1Name=dyn_l_FrontVest_01_jnt, bone2Name=dyn_l_FrontVest_01_MSTR, swing1LimitDegrees=15, swing2LimitDegrees=15, twistLimitDegrees=2}
 {bone1Name=dyn_l_FrontVest_02_jnt, bone2Name=dyn_l_FrontVest_01_jnt, swing1LimitDegrees=15, swing2LimitDegrees=15, twistLimitDegrees=2}
 {bone1Name=dyn_l_SideVest_01_jnt, bone2Name=dyn_l_SideVest_01_MSTR, swing1LimitDegrees=15, swing2LimitDegrees=15, twistLimitDegrees=2}
 {bone1Name=dyn_l_SideVest_02_jnt, bone2Name=dyn_l_SideVest_01_jnt, swing1LimitDegrees=15, swing2LimitDegrees=15, twistLimitDegrees=2}
 {bone1Name=dyn_r_BackVest_01_jnt, bone2Name=dyn_r_BackVest_01_MSTR, swing1LimitDegrees=15, swing2LimitDegrees=15, twistLimitDegrees=2}
 {bone1Name=dyn_r_BackVest_02_jnt, bone2Name=dyn_r_BackVest_01_jnt, swing1LimitDegrees=15, swing2LimitDegrees=15, twistLimitDegrees=2}
 {bone1Name=dyn_r_FrontVest_01_jnt, bone2Name=dyn_r_FrontVest_01_MSTR, swing1LimitDegrees=15, swing2LimitDegrees=15, twistLimitDegrees=2}
 {bone1Name=dyn_r_FrontVest_02_jnt, bone2Name=dyn_r_FrontVest_01_jnt, swing1LimitDegrees=15, swing2LimitDegrees=15, twistLimitDegrees=2}
 {bone1Name=dyn_r_SideVest_01_jnt, bone2Name=dyn_r_SideVest_01_MSTR, swing1LimitDegrees=15, swing2LimitDegrees=15, twistLimitDegrees=2}
 {bone1Name=dyn_r_SideVest_02_jnt, bone2Name=dyn_r_SideVest_01_jnt, swing1LimitDegrees=15, swing2LimitDegrees=15, twistLimitDegrees=2}
 {
  bone1Name=dyn_c_BackVest_01_MSTR
  bone2Name=spine_03
  swing1LimitDegrees=45
  swing1Motion=Free
  swing2LimitDegrees=45
  swing2Motion=Free
  twistLimitDegrees=45
  twistMotion=Free
 }
 {
  bone1Name=dyn_l_BackVest_01_MSTR
  bone2Name=spine_03
  swing1LimitDegrees=45
  swing1Motion=Free
  swing2LimitDegrees=45
  swing2Motion=Free
  twistLimitDegrees=45
  twistMotion=Free
 }
 {
  bone1Name=dyn_l_SideVest_01_MSTR
  bone2Name=spine_03
  swing1LimitDegrees=45
  swing1Motion=Free
  swing2LimitDegrees=45
  swing2Motion=Free
  twistLimitDegrees=45
  twistMotion=Free
 }
 {
  bone1Name=dyn_l_FrontVest_01_MSTR
  bone2Name=spine_03
  swing1LimitDegrees=45
  swing1Motion=Free
  swing2LimitDegrees=45
  swing2Motion=Free
  twistLimitDegrees=45
  twistMotion=Free
 }
 {
  bone1Name=dyn_r_FrontVest_01_MSTR
  bone2Name=spine_03
  swing1LimitDegrees=45
  swing1Motion=Free
  swing2LimitDegrees=45
  swing2Motion=Free
  twistLimitDegrees=45
  twistMotion=Free
 }
 {
  bone1Name=dyn_r_SideVest_01_MSTR
  bone2Name=spine_03
  swing1LimitDegrees=45
  swing1Motion=Free
  swing2LimitDegrees=45
  swing2Motion=Free
  twistLimitDegrees=45
  twistMotion=Free
 }
 {
  bone1Name=dyn_r_BackVest_01_MSTR
  bone2Name=spine_03
  swing1LimitDegrees=45
  swing1Motion=Free
  swing2LimitDegrees=45
  swing2Motion=Free
  twistLimitDegrees=45
  twistMotion=Free
 }
 {
  bone1Name=spine_03
  bone2Name=spine_02
  swing1LimitDegrees=45
  swing1Motion=Free
  swing2LimitDegrees=45
  swing2Motion=Free
  twistLimitDegrees=45
  twistMotion=Free
 }
 {
  bone1Name=spine_02
  bone2Name=spine_01
  swing1LimitDegrees=45
  swing1Motion=Free
  swing2LimitDegrees=45
  swing2Motion=Free
  twistLimitDegrees=45
  twistMotion=Free
 }
]
