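# Post-process style AnimBP for the DriftAvatar male body cosmetic, adding secondary motion to vest, collar, pants and shoe details.
# Copies the parent mesh pose, then chains bone constraints, Control Rig, vest RigidBody, springs, AnimDynamics and a throw-aim collar offset.
# EventGraph derives weapon and emote rig inputs, damps vest physics by movement state, and disables springs at high speed.
# Outside the frontend, it feeds leg motion and skydive wind to WPO cloth materials.
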AnimBlueprint /BRCosmetics/Characters/Player/Male/Medium/Bodies/M_MED_DriftAvatar/Meshes/M_MED_DriftAvatar_AnimBP
target_skeleton=&/M_MED_DriftAvatar_Skeleton
graphs [
 AnimGraph {nodes [
   n1 = "Constraint - collar_r_jnt (RR)"!(ComponentPose, Alpha=1, ConstraintWeights_0=0.7, ConstraintWeights_1:float=0.3) -> (Pose) {
    alphaBoolBlend {blendOption=Linear}
    boneToModify {boneName=collar_r_jnt}
    constraintSetup [{transformType=Rotation, targetBone {boneName=neck_01}}, {transformType=Rotation, targetBone {boneName=spine_05}}]
    constraintWeights [1,1]
   }
   n2 = "Constraint - collar_c_jnt (RR)"!(ComponentPose=@n1, Alpha=1, ConstraintWeights_0=0.7, ConstraintWeights_1:float=0.3) -> (Pose) {
    alphaBoolBlend {blendOption=Linear}
    boneToModify {boneName=collar_c_jnt}
    constraintSetup [{transformType=Rotation, targetBone {boneName=neck_01}}, {transformType=Rotation, targetBone {boneName=spine_05}}]
    constraintWeights [1,1]
   }
   n3 = "Constraint - collar_l_jnt (RR)"!(ComponentPose=@n2, Alpha=1, ConstraintWeights_0=0.7, ConstraintWeights_1:float=0.3) -> (Pose) {
    alphaBoolBlend {blendOption=Linear}
    boneToModify {boneName=collar_l_jnt}
    constraintSetup [{transformType=Rotation, targetBone {boneName=neck_01}}, {transformType=Rotation, targetBone {boneName=spine_05}}]
    constraintWeights [1,1]
   }
   n4 = "Constraint - collar_frt_r_jnt (PP)"!(ComponentPose=@n3, Alpha=1, ConstraintWeights_0=0.5, ConstraintWeights_1:float=0.5) -> (Pose) {
    alphaBoolBlend {blendOption=Linear}
    boneToModify {boneName=collar_frt_r_jnt}
    constraintSetup [{targetBone {boneName=neck_01}}, {targetBone {boneName=spine_05}}]
    constraintWeights [1,1]
   }
   "Constraint - collar_frt_l_jnt (PP)"!(ComponentPose=@n4, Alpha=1, ConstraintWeights_0=0.5, ConstraintWeights_1:float=0.5) -> (Pose) {
    alphaBoolBlend {blendOption=Linear}
    boneToModify {boneName=collar_frt_l_jnt}
    constraintSetup [{targetBone {boneName=neck_01}}, {targetBone {boneName=spine_05}}]
    constraintWeights [1,1]
   }
   n5 = "Constraint - dyn_r_BackVest_01_MSTR (PP)"!(
    ComponentPose=Local To Component(LocalPose=Copy Pose From Mesh() {bUseAttachedParent=true})
    Alpha=1
    ConstraintWeights_0=0.9
    ConstraintWeights_1:float=0.1
   ) -> (Pose) {
    alphaBoolBlend {blendOption=Linear}
    boneToModify {boneName=dyn_r_BackVest_01_MSTR}
    constraintSetup [{targetBone {boneName=spine_04}}, {targetBone {boneName=spine_05}}]
    constraintWeights [1,1]
   }
   n6 = "Constraint - dyn_r_SideVest_01_MSTR (PPP)"!(ComponentPose=@n5, Alpha=1, ConstraintWeights_0=0.7, ConstraintWeights_1:float=0.1, ConstraintWeights_2:float=0.2) -> (Pose) {
    alphaBoolBlend {blendOption=Linear}
    boneToModify {boneName=dyn_r_SideVest_01_MSTR}
    constraintSetup [{targetBone {boneName=spine_04}}, {targetBone {boneName=spine_05}}, {targetBone {boneName=clavicle_r}}]
    constraintWeights [1,1,1]
   }
   n7 = "Constraint - dyn_r_FrontVest_01_MSTR (PP)"!(ComponentPose=@n6, Alpha=1, ConstraintWeights_0=0.9, ConstraintWeights_1:float=0.1) -> (Pose) {
    alphaBoolBlend {blendOption=Linear}
    boneToModify {boneName=dyn_r_FrontVest_01_MSTR}
    constraintSetup [{targetBone {boneName=spine_04}}, {targetBone {boneName=spine_05}}]
    constraintWeights [1,1]
   }
   n8 = "Constraint - dyn_c_BackVest_01_MSTR (PP)"!(ComponentPose=@n7, Alpha=1, ConstraintWeights_0=0.9, ConstraintWeights_1:float=0.1) -> (Pose) {
    alphaBoolBlend {blendOption=Linear}
    boneToModify {boneName=dyn_c_BackVest_01_MSTR}
    constraintSetup [{targetBone {boneName=spine_04}}, {targetBone {boneName=spine_05}}]
    constraintWeights [1,1]
   }
   n9 = "Constraint - dyn_l_BackVest_01_MSTR (PP)"!(ComponentPose=@n8, Alpha=1, ConstraintWeights_0=0.9, ConstraintWeights_1:float=0.1) -> (Pose) {
    alphaBoolBlend {blendOption=Linear}
    boneToModify {boneName=dyn_l_BackVest_01_MSTR}
    constraintSetup [{targetBone {boneName=spine_04}}, {targetBone {boneName=spine_05}}]
    constraintWeights [1,1]
   }
   n10 = "Constraint - dyn_l_SideVest_01_MSTR (PPP)"!(ComponentPose=@n9, Alpha=1, ConstraintWeights_0=0.7, ConstraintWeights_1:float=0.1, ConstraintWeights_2:float=0.2) -> (Pose) {
    alphaBoolBlend {blendOption=Linear}
    boneToModify {boneName=dyn_l_SideVest_01_MSTR}
    constraintSetup [{targetBone {boneName=spine_04}}, {targetBone {boneName=spine_05}}, {targetBone {boneName=clavicle_l}}]
    constraintWeights [1,1,1]
   }
   n11 = "Constraint - dyn_l_FrontVest_01_MSTR (PP)"!(ComponentPose=@n10, Alpha=1, ConstraintWeights_0=0.9, ConstraintWeights_1:float=0.1) -> (Pose) {
    alphaBoolBlend {blendOption=Linear}
    boneToModify {boneName=dyn_l_FrontVest_01_MSTR}
    constraintSetup [{targetBone {boneName=spine_04}}, {targetBone {boneName=spine_05}}]
    constraintWeights [1,1]
   }
   n12 = "Constraint - collar_r_MSTR (RRR)"!(ComponentPose=@n11, Alpha=1, ConstraintWeights_0=0.25, ConstraintWeights_1:float=0.25, ConstraintWeights_2:float=0.5) -> (Pose) {
    alphaBoolBlend {blendOption=Linear}
    boneToModify {boneName=collar_r_MSTR}
    constraintSetup [
     {transformType=Rotation, targetBone {boneName=spine_05}}
     {transformType=Rotation, targetBone {boneName=clavicle_r}}
     {transformType=Rotation, targetBone {boneName=neck_01}}
    ]
    constraintWeights [1,1,1]
   }
   n13 = "Constraint - collar_c_MSTR (RR)"!(ComponentPose=@n12, Alpha=1, ConstraintWeights_0=0.5, ConstraintWeights_1:float=0.5) -> (Pose) {
    alphaBoolBlend {blendOption=Linear}
    boneToModify {boneName=collar_c_MSTR}
    constraintSetup [{transformType=Rotation, targetBone {boneName=spine_05}}, {transformType=Rotation, targetBone {boneName=neck_01}}]
    constraintWeights [1,1]
   }
   n14 = "Constraint - collar_l_MSTR (RRR)"!(ComponentPose=@n13, Alpha=1, ConstraintWeights_0=0.25, ConstraintWeights_1:float=0.25, ConstraintWeights_2:float=0.5) -> (Pose) {
    alphaBoolBlend {blendOption=Linear}
    boneToModify {boneName=collar_l_MSTR}
    constraintSetup [
     {transformType=Rotation, targetBone {boneName=spine_05}}
     {transformType=Rotation, targetBone {boneName=clavicle_l}}
     {transformType=Rotation, targetBone {boneName=neck_01}}
    ]
    constraintWeights [1,1,1]
   }
   n15 = "Constraint - collar_frt_r_MSTR (PP)"!(ComponentPose=@n14, Alpha=1, ConstraintWeights_0=0.5, ConstraintWeights_1:float=0.5) -> (Pose) {
    alphaBoolBlend {blendOption=Linear}
    boneToModify {boneName=collar_frt_r_MSTR}
    constraintSetup [{targetBone {boneName=neck_01}}, {targetBone {boneName=clavicle_r}}]
    constraintWeights [1,1]
   }
   n16 = "Constraint - collar_frt_l_MSTR (PP)"!(ComponentPose=@n15, Alpha=1, ConstraintWeights_0=0.5, ConstraintWeights_1:float=0.5) -> (Pose) {
    alphaBoolBlend {blendOption=Linear}
    boneToModify {boneName=collar_frt_l_MSTR}
    constraintSetup [{targetBone {boneName=neck_01}}, {targetBone {boneName=clavicle_l}}]
    constraintWeights [1,1]
   }
   n17 = Control Rig!(
    Source=Component To Local(ComponentPose=@n16)
    Alpha
    IsPistol:double=Get Is Pistol Alpha() -> (Is Pistol Alpha:double)
    IsRifle:double=Get Is Rifle Alpha() -> (Is Rifle Alpha:double)
    Is not playing emote Alpha:double=Get Is not Playing Emote Alpha() -> (Is not Playing Emote Alpha:double)
   ) -> (Pose) {
    alphaBoolBlend {blendOption=Linear}
    controlRigAssetReference {blueprintRigClass=&/M_MED_DriftAvatar_CtrlRig.M_MED_DriftAvatar_CtrlRig_C}
   }
   n18 = RigidBody!(
    ComponentPose=Local To Component(LocalPose=@n17)
    Alpha=Get VestAlpha() -> (VestAlpha:double)
    OverrideWorldGravity:Vector=Get GravityOverride()
    ExternalForce
   ) -> (Pose) {
    bClampLinearTranslationLimitToRefPose=true
    bOverrideWorldGravity=true
    overridePhysicsAsset=&/M_MED_DriftAvatar_Physics
    simulationSpace=BaseBoneSpace
    alphaBoolBlend {blendOption=Linear}
    alphaScaleBiasClamp {interpSpeedIncreasing=2}
    baseBoneRef {boneName=spine_03}
    componentAppliedLinearAccClamp=<3000 3000 3000>
    componentLinearAccScale=<0.4 0.3 0.3>
    componentLinearVelScale=<0.4 0.3 0.3>
    simSpaceSettings {
     maxAngularAcceleration=1000
     maxAngularVelocity=3
     maxLinearAcceleration=700
     maxLinearVelocity=500
     velocityScaleZ=0
     worldAlpha=0.25
     externalAngularVelocity {z=0.1}
     externalLinearDragV=<0.2 0.2>
    }
   }
   n19 = "Constraint - Shoetag_r_MSTR (PP)"!(ComponentPose=@n18, Alpha=1, ConstraintWeights_0=0.2, ConstraintWeights_1:float=0.8) -> (Pose) {
    alphaBoolBlend {blendOption=Linear}
    boneToModify {boneName=Shoetag_r_MSTR}
    constraintSetup [{targetBone {boneName=calf_twist_01_r}}, {targetBone {boneName=foot_r}}]
    constraintWeights [1,1]
   }
   n20 = "Constraint - Shoetag_l_MSTR (PP)"!(ComponentPose=@n19, Alpha=1, ConstraintWeights_0=0.2, ConstraintWeights_1:float=0.8) -> (Pose) {
    alphaBoolBlend {blendOption=Linear}
    boneToModify {boneName=Shoetag_l_MSTR}
    constraintSetup [{targetBone {boneName=calf_twist_01_l}}, {targetBone {boneName=foot_l}}]
    constraintWeights [1,1]
   }
   n21 = Spring controller - Bone: ShoeTongue_l_jnt!(ComponentPose:Component Space Pose Link Structure=@n20, Alpha:float=Get SpringAlpha() -> (SpringAlpha:double)) -> (Pose:Component Space Pose Link Structure) {
    is=title AnimGraphNode_SpringBone
    bRotateX=true
    bRotateY=true
    bRotateZ=true
    springDamping=3.5
    springStiffness=200
    alphaBoolBlend {blendOption=Linear}
    springBone {boneName=ShoeTongue_l_jnt}
   }
   n22 = Spring controller - Bone: ShoeTongue_r_jnt!(ComponentPose:Component Space Pose Link Structure=@n21, Alpha:float=Get Spring Alpha 02() -> (Spring Alpha 02:double)) -> (Pose:Component Space Pose Link Structure) {
    is=title AnimGraphNode_SpringBone
    bRotateX=true
    bRotateY=true
    bRotateZ=true
    springDamping=3.25
    springStiffness=150
    alphaBoolBlend {blendOption=Linear}
    springBone {boneName=ShoeTongue_r_jnt}
   }
   n23 = Anim Dynamics - Bone: ShoeTongue_l_jnt!(ComponentPose=@n22, Alpha=0.5, GravityOverride:Vector="0,0,-980", ExternalForce="0,0,0") -> (Pose) {
    angularDampingOverride=1
    angularSpringConstant=800
    bAngularSpring=true
    bOverrideAngularDamping=true
    bUseGravityOverride=true
    linearDampingOverride=0.1
    simulationSpace=RootRelative
    alphaBoolBlend {blendOption=Linear}
    boundBone {boneName=ShoeTongue_l_jnt}
    componentAppliedLinearAccClamp=<3000 3000 3000>
    componentLinearAccScale=<0.5 0.5 0.5>
    componentLinearVelScale=<0.5 0.5 0.5>
    gravityOverride {z=-980}
    physicsBodyDefinitions [{
      boundBone {boneName=ShoeTongue_l_jnt}
      boxExtents=<10 10 10>
      constraintSetup {angularLimitsMax=<10 20 10>, angularLimitsMin=<-10 -20 -10>, angularTarget=<1 0 0>}
      localJointOffset=<4 0 0>
     }]
    relativeSpaceBone {boneName=root}
   }
   n24 = Anim Dynamics - Bone: ShoeTongue_r_jnt!(ComponentPose=@n23, Alpha=0.5, GravityOverride:Vector="0,0,-980", ExternalForce="0,0,0") -> (Pose) {
    angularDampingOverride=1
    angularSpringConstant=800
    bAngularSpring=true
    bOverrideAngularDamping=true
    bUseGravityOverride=true
    linearDampingOverride=0.1
    simulationSpace=RootRelative
    alphaBoolBlend {blendOption=Linear}
    boundBone {boneName=ShoeTongue_r_jnt}
    componentAppliedLinearAccClamp=<3000 3000 3000>
    componentLinearAccScale=<0.5 0.5 0.5>
    componentLinearVelScale=<0.5 0.5 0.5>
    gravityOverride {z=-980}
    physicsBodyDefinitions [{
      boundBone {boneName=ShoeTongue_r_jnt}
      boxExtents=<10 10 10>
      constraintSetup {angularLimitsMax=<10 20 10>, angularLimitsMin=<-10 -20 -10>, angularTarget=<1 0 0>}
      localJointOffset=<4 0 0>
     }]
    relativeSpaceBone {boneName=root}
   }
   n25 = Spring controller - Bone: Shoetag_r_jnt!(ComponentPose:Component Space Pose Link Structure=@n24, Alpha:float=Get Spring Alpha 04() -> (Spring Alpha 04:double)) -> (Pose:Component Space Pose Link Structure) {
    is=title AnimGraphNode_SpringBone
    bRotateX=true
    bRotateY=true
    bRotateZ=true
    springDamping=3.5
    springStiffness=200
    alphaBoolBlend {blendOption=Linear}
    springBone {boneName=Shoetag_r_jnt}
   }
   n26 = Spring controller - Bone: Shoetag_l_jnt!(ComponentPose:Component Space Pose Link Structure=@n25, Alpha:float=Get Spring Alpha 03() -> (Spring Alpha 03:double)) -> (Pose:Component Space Pose Link Structure) {
    is=title AnimGraphNode_SpringBone
    bRotateX=true
    bRotateY=true
    bRotateZ=true
    springDamping=3.25
    springStiffness=150
    alphaBoolBlend {blendOption=Linear}
    springBone {boneName=Shoetag_l_jnt}
   }
   n27 = Anim Dynamics - Bone: Shoetag_r_jnt!(ComponentPose=@n26, Alpha=0.5, GravityOverride:Vector="0,0,-980", ExternalForce="0,0,0") -> (Pose) {
    angularDampingOverride=1
    angularSpringConstant=1800
    bAngularSpring=true
    bOverrideAngularDamping=true
    bUseGravityOverride=true
    simulationSpace=RootRelative
    alphaBoolBlend {blendOption=Linear}
    boundBone {boneName=Shoetag_r_jnt}
    componentAppliedLinearAccClamp=<3000 3000 3000>
    componentLinearAccScale=<0.5 0.5 0.5>
    componentLinearVelScale=<0.5 0.5 0.5>
    gravityOverride {z=-980}
    physicsBodyDefinitions [{
      boundBone {boneName=Shoetag_r_jnt}
      boxExtents=<10 10 10>
      constraintSetup {angularLimitsMax=<10 20 20>, angularLimitsMin=<-10 -20 -20>, angularTarget=<1 0 0>}
      localJointOffset=<6 0 0>
     }]
    relativeSpaceBone {boneName=root}
   }
   n28 = Anim Dynamics - Bone: Shoetag_l_jnt!(ComponentPose=@n27, Alpha=0.5, GravityOverride:Vector="0,0,-980", ExternalForce="0,0,0") -> (Pose) {
    angularDampingOverride=1
    angularSpringConstant=1800
    bAngularSpring=true
    bOverrideAngularDamping=true
    bUseGravityOverride=true
    linearDampingOverride=0.1
    simulationSpace=RootRelative
    alphaBoolBlend {blendOption=Linear}
    boundBone {boneName=Shoetag_l_jnt}
    componentAppliedLinearAccClamp=<3000 3000 3000>
    componentLinearAccScale=<0.5 0.5 0.5>
    componentLinearVelScale=<0.5 0.5 0.5>
    gravityOverride {z=-980}
    physicsBodyDefinitions [{
      boundBone {boneName=Shoetag_l_jnt}
      boxExtents=<10 10 10>
      constraintSetup {angularLimitsMax=<10 20 20>, angularLimitsMin=<-10 -20 -20>, angularTarget=<1 0 0>}
      localJointOffset=<6 0 0>
     }]
    relativeSpaceBone {boneName=root}
   }
   n29 = Get Spring Alpha 04()
   n30 = Spring controller - Bone: dyn_l_upperpants_01_jnt!(ComponentPose:Component Space Pose Link Structure=@n28, Alpha:float=@n29) -> (Pose:Component Space Pose Link Structure) {
    is=title AnimGraphNode_SpringBone
    bRotateX=true
    bRotateY=true
    bRotateZ=true
    springDamping=2.5
    springStiffness=200
    alphaBoolBlend {blendOption=Linear}
    springBone {boneName=dyn_l_upperpants_01_jnt}
   }
   n31 = Get Spring Alpha 03()
   n32 = Spring controller - Bone: dyn_l_upperpants_02_jnt!(ComponentPose:Component Space Pose Link Structure=@n30, Alpha:float=@n31) -> (Pose:Component Space Pose Link Structure) {
    is=title AnimGraphNode_SpringBone
    bRotateX=true
    bRotateY=true
    bRotateZ=true
    springDamping=3
    springStiffness=250
    alphaBoolBlend {blendOption=Linear}
    springBone {boneName=dyn_l_upperpants_02_jnt}
   }
   n33 = Spring controller - Bone: dyn_r_upperpants_02_jnt!(ComponentPose:Component Space Pose Link Structure=@n32, Alpha:float=@n29) -> (Pose:Component Space Pose Link Structure) {
    is=title AnimGraphNode_SpringBone
    bRotateX=true
    bRotateY=true
    bRotateZ=true
    springDamping=2.5
    springStiffness=200
    alphaBoolBlend {blendOption=Linear}
    springBone {boneName=dyn_r_upperpants_02_jnt}
   }
   n34 = Spring controller - Bone: dyn_r_upperpants_01_jnt!(ComponentPose:Component Space Pose Link Structure=@n33, Alpha:float=@n31) -> (Pose:Component Space Pose Link Structure) {
    is=title AnimGraphNode_SpringBone
    bRotateX=true
    bRotateY=true
    bRotateZ=true
    springDamping=3
    springStiffness=250
    alphaBoolBlend {blendOption=Linear}
    springBone {boneName=dyn_r_upperpants_01_jnt}
   }
   n35 = "Transform (Modify) Bone - Bone: collar_r_MSTR"(
    ComponentPose=@n34
    Alpha=Get Is Targeting Consumable Throw Alpha() -> (Is Targeting Consumable Throw Alpha:double)
    Translation="0,0,1"
    Rotation="20,0,0"
    Scale="1,1,1"
   ) {
    rotationMode=BMM_Additive
    rotationSpace=BCS_BoneSpace
    translationMode=BMM_Additive
    translationSpace=BCS_BoneSpace
    alphaBoolBlend {blendOption=Linear}
    alphaScaleBiasClamp {bInterpResult=true, interpSpeedDecreasing=4}
    boneToModify {boneName=collar_r_MSTR}
    rotation {pitch=20}
    translation {z=1}
   }
   Output Pose(Result=Component To Local(ComponentPose=@n35)) {layerGroup=None}
  ]}
 EventGraph {nodes [
   n1 = Event BlueprintUpdateAnimation()
   > n2 = Sequence!(execute) -> (
    then_0
    then_1
    then_2:Exec
    then_3:Exec
    then_4:Exec
    then_5:Exec
    then_6:Exec
    then_7:Exec
    then_8:Exec
    then_9:Exec
    then_10:Exec
    then_11:Exec
    then_12:Exec
    then_13:Exec
   )
   > Is Valid(InputObject=Get Player Pawn() -> (Player Pawn:Fort Player Pawn Athena Object Reference))
   > n3 = Branch(Condition=Get bIsInFrontEnd())
   Set WPO Globals Function(
    execute=@n3.else
    Material=Get Mat Instance Dynamic()
    Overlay Material=Get Overlay Mat Instance Dynamic()
    Skeletal Mesh=Get Skel Mesh()
    "Local Noise Transform  Location"=Get Local Noise Transform Location() -> (Local Noise Transform Location:Vector)
    Local Noise Offset=Get Local Noise Offset()
   )
   > Set WPO Regional Function!(
    execute
    Material:Material Instance Dynamic Object Reference=Get Mat Instance Dynamic()
    Overlay Material:Material Instance Dynamic Object Reference=Get Overlay Mat Instance Dynamic()
    Skeletal Mesh:Skeletal Mesh Component Object Reference=Get Skel Mesh()
    Velocity Transform:Name=calf_r
    Velocity Interp Speed:double=Get Velocity Interp Speed()
    Collision Transform:Name=thigh_r
    Collision Transform Axis:EAxis Enum
    Invert Collision Transform Axis:bool=false
    Delta Time X:double=@n1.DeltaTimeX
    Previous Velocity Transform Location:"Vector (by ref)"=Get R1 Velocity Transform Location()
    Previous Velocity Transform Velocity:"Vector (by ref)"=Get R1 Velocity Transform Velocity()
    Region:int=1
   ) -> (then)
   > Set WPO Regional Function(
    Material=Get Mat Instance Dynamic()
    Overlay Material=Get Overlay Mat Instance Dynamic()
    Skeletal Mesh=Get Skel Mesh()
    Velocity Transform=calf_l
    Velocity Interp Speed=Get Velocity Interp Speed()
    Collision Transform=thigh_l
    Invert Collision Transform Axis=true
    Delta Time X=@n1.DeltaTimeX
    Previous Velocity Transform Location=Get R2 Velocity Transform Location()
    Previous Velocity Transform Velocity=Get R2 Velocity Transform Velocity()
    Region=2
   )
   > n4 = Branch(Condition=Get bIsSkyDiving())
   > n5 = Set WPO Pho Wind Intensity(WPO Pho Wind Intensity=1)
   SetScalarParameterValue(
    execute=[@n6.then, @n5.then]
    self=[Get Mat Instance Dynamic(), Get Overlay Mat Instance Dynamic()]
    ParameterName=ClothSim_WindStrength
    Value=Get WPO Pho Wind Intensity()
   )
   n6 = Set WPO Pho Wind Intensity(execute=@n4.else, WPO Pho Wind Intensity=0)
   n7 = Gravity Override State Function DEPRECATED(
    execute=@n2.then_1
    CharacterPartAnimInstance=Self-Reference()
    GameTimeInSeconds=GetGameTimeInSeconds()
    GravityOverrideParams=Get Gravity Override Params()
   )
   > Set GravityOverride(GravityOverride=@n7)
   Set VestAlpha(
    execute=@n2.then_2
    VestAlpha:double=float * float!(
     A=Get Is Sky Diving Vest Alpha() -> (Is Sky Diving Vest Alpha:double)
     B=Get Is Tactical Sprinting Alpha() -> (Is Tactical Sprinting Alpha:double)
     C:float=Get Is Sliding Alpha() -> (Is Sliding Alpha:float)
     D:double=Get Jumping Vest Alpha() -> (Jumping Vest Alpha:double)
     E:double=Get Is Swim Sprinting Alpha() -> (Is Swim Sprinting Alpha:double)
     F:double=Get Is Surface Swimming Alpha() -> (Is Surface Swimming Alpha:double)
     G:double=1
     H:float=Get Is Crouching Alpha() -> (Is Crouching Alpha:float)
    ) -> (ReturnValue) {is=title K2Node_PromotableOperator}
   ) -> (Output_Get:double)
   n8 = Sequence!(execute=@n2.then_3) -> (then_0, then_1, then_2, then_3, then_4, then_5, then_6)
   > n9 = Branch(Condition=Get bIsSkyDiving())
   > Set Is Sky Diving Vest Alpha(Is Sky Diving Vest Alpha:double=0.5) -> (Output_Get:double)
   Set Is Sky Diving Vest Alpha(execute=@n9.else, Is Sky Diving Vest Alpha=1)
   n10 = Get bIsCrouching()
   n11 = Get bIsTacticalSprinting()
   n12 = Get bIsSprinting()
   n13 = Branch(execute=@n8.then_1, Condition=AND Boolean(A=OR Boolean(A=@n11, B=@n12), B=NOT Boolean(A=@n10)))
   > Set Is Tactical Sprinting Alpha(Is Tactical Sprinting Alpha:double=0.25) -> (Output_Get:double)
   Set Is Tactical Sprinting Alpha(execute=@n13.else, Is Tactical Sprinting Alpha=1)
   n14 = Branch(execute=@n8.then_2, Condition=Get bIsSliding())
   > Set Is Sliding Alpha(Is Sliding Alpha:float=0.25) -> (Output_Get:float)
   Set Is Sliding Alpha(execute=@n14.else, Is Sliding Alpha=1)
   n15 = NOT Boolean(A=Get bIsSkyDiving())
   n16 = Get bIsJumping()
   n17 = NOT Boolean(A=Get bIsSkyDiving())
   n18 = Get bIsFalling()
   n19 = Branch(execute=@n8.then_3, Condition=OR Boolean(A=AND Boolean(A=@n16, B=@n15), B=AND Boolean(A=@n18, B=@n17)))
   > Set Jumping Vest Alpha(Jumping Vest Alpha:double=0.75) -> (Output_Get:double)
   Set Jumping Vest Alpha(execute=@n19.else, Jumping Vest Alpha=1)
   n20 = Branch(execute=@n8.then_4, Condition=Get bIsSurfaceSwimming())
   > Set Is Surface Swimming Alpha(Is Surface Swimming Alpha:double=1) -> (Output_Get:double)
   Set Is Surface Swimming Alpha(execute=@n20.else, Is Surface Swimming Alpha=1)
   n21 = Branch(execute=@n8.then_5, Condition=Get bIsSwimSprinting() -> (bIsSwimSprinting:bool))
   > Set Is Swim Sprinting Alpha(Is Swim Sprinting Alpha:double=1) -> (Output_Get:double)
   Set Is Swim Sprinting Alpha(execute=@n21.else, Is Swim Sprinting Alpha=1)
   n22 = Branch(execute=@n8.then_6, Condition=Get bIsCrouching())
   > Set Is Crouching Alpha(Is Crouching Alpha:float=1) -> (Output_Get:float)
   Set Is Crouching Alpha(execute=@n22.else, Is Crouching Alpha=1)
   n23 = Sequence!(execute=@n2.then_4) -> (then_0, then_1, then_2, then_3, then_4)
   > n24 = Cast To FortPlayerAnimInstance(Object=GetAnimInstance(self=GetOwnerSkeletalMeshComponent()))
   > Set isPistol(isPistol=false)
   > Set isRifle(isRifle:bool=false) -> (Output_Get:bool)
   > n25 = Switch on EFortWeaponCoreAnimation!(execute, Selection:EFortWeaponCoreAnimation Enum=Get WeaponCoreAnim(self=@n24)) -> (
    Melee:Exec
    Pistol:Exec
    Shotgun:Exec
    PaperBlueprint:Exec
    Rifle:Exec
    MeleeOneHand:Exec
    MachinePistol:Exec
    RocketLauncher:Exec
    GrenadeLauncher:Exec
    GoingCommando:Exec
    AssaultRifle:Exec
    TacticalShotgun:Exec
    SniperRifle:Exec
    TrapPlacement:Exec
    ShoulderLauncher:Exec
    AbilityDecoTool:Exec
    Crossbow:Exec
    C4:Exec
    RemoteControl:Exec
    DualWield:Exec
    AR_BullPup:Exec
    AR_ForwardGrip:Exec
    MedPackPaddles:Exec
    SMG_P90:Exec
    AR_DrumGun:Exec
    Consumable_Small:Exec
    Consumable_Large:Exec
    Balloon:Exec
    MountedTurret:Exec
    CreativeTool:Exec
    ExplosiveBow:Exec
    AshtonIndigo:Exec
    AshtonChicago:Exec
    MeleeDualWield:Exec
    Unarmed:Exec
   )
   Set isRifle(execute=[@n25.AssaultRifle, @n25.Rifle, @n25.RocketLauncher, @n25.Shotgun, @n25.SniperRifle, @n25.TacticalShotgun], isRifle=true)
   Set isPistol(execute=@n25.Pistol, isPistol=true)
   Set Is Rifle Alpha(execute=@n23.then_1, Is Rifle Alpha:double="To Float (Boolean)"(InBool=Get isRifle() -> (isRifle:bool))) -> (Output_Get:double)
   Set Is Pistol Alpha(execute=@n23.then_2, Is Pistol Alpha:double="To Float (Boolean)"(InBool=Get isPistol() -> (isPistol:bool))) -> (Output_Get:double)
   n26 = Set pawnVelocitySpeed(execute=@n2.then_5, pawnVelocitySpeed:double=Vector Length(A=Get PawnVelocity())) -> (Output_Get:double)
   > n27 = Set pawnVelocitySpeedCutoff(pawnVelocitySpeedCutoff:double=SelectFloat(A=0, B=1, bPickA="float > float"(A=@n26, B=1200) {is=title K2Node_CallFunction})) -> (Output_Get:double)
   > n28 = Set SpringAlpha(SpringAlpha:double=float * float!(A=0.008, B=@n27) -> (ReturnValue)) -> (Output_Get:double)
   > Set Spring Alpha 02(Spring Alpha 02:double=float * float(A=0.75, B=@n28)) -> (Output_Get:double)
   > n29 = Set Spring Alpha 03(Spring Alpha 03:double=float * float(A=0.005, B=@n27)) -> (Output_Get:double)
   > Set Spring Alpha 04(Spring Alpha 04:double=float * float(A=0.5, B=@n29)) -> (Output_Get:double)
   n30 = Get bIsPlayingEmote()
   Set Is not Playing Emote Alpha(execute=@n2.then_6, Is not Playing Emote Alpha:double="To Float (Boolean)"(InBool=NOT Boolean(A=@n30))) -> (Output_Get:double)
   Set Is Targeting Consumable Throw Alpha(
    execute=@n2.then_7
    Is Targeting Consumable Throw Alpha:double="To Float (Boolean)"(InBool=Get bIsTargetingConsumableThrow() -> (bIsTargetingConsumableThrow:bool))
   ) -> (Output_Get:double)
   Event BlueprintInitializeAnimation()
   > Set Skel Mesh(Skel Mesh=GetOwnerSkeletalMeshComponent())
   n31 = TryGetPawnOwner()
   > Is Valid(InputObject=@n31)
   > n32 = Cast To FortPlayerPawnAthena(Object=@n31)
   > Set Player Pawn(Player Pawn:Fort Player Pawn Athena Object Reference=@n32) -> (Output_Get:Fort Player Pawn Athena Object Reference)
   Event OnCharacterCustomizationLoaded()
   > n33 = Set Skel Mesh(Skel Mesh=GetPartSkeletalMeshComponent())
   > n34 = Get Dynamic Material Instance(SkeletalMeshComponent=@n33)
   > Set Mat Instance Dynamic(Mat Instance Dynamic=@n34)
   > n35 = GetOrCreateDynamicOverlayMaterial(MeshComponent=@n33, MaterialName=None)
   > Set Overlay Mat Instance Dynamic(Overlay Mat Instance Dynamic=@n35)
  ]}
]
variables [
 Gravity Override Params {type=GravityOverrideParamsStruct}
 GravityOverride {type=Vector}
 Is Crouching Alpha {type=number}
 Is not Playing Emote Alpha {type=number}
 Is Pistol Alpha {type=number}
 Is Rifle Alpha {type=number}
 Is Sky Diving Vest Alpha {type=number}
 Is Sliding Alpha {type=number}
 Is Surface Swimming Alpha {type=number}
 Is Swim Sprinting Alpha {type=number}
 Is Tactical Sprinting Alpha {type=number}
 Is Targeting Consumable Throw Alpha {type=number}
 isPistol {type=boolean}
 isRifle {type=boolean}
 Jumping Vest Alpha {type=number}
 Local Noise Offset {type=Vector}
 Local Noise Transform Location {type=Vector}
 Mat Instance Dynamic {type=MaterialInstanceDynamic}
 Overlay Mat Instance Dynamic {type=MaterialInstanceDynamic}
 pawnVelocitySpeed {type=number}
 pawnVelocitySpeedCutoff {type=number}
 Player Pawn {type=FortPlayerPawnAthena}
 R1 Velocity Transform Location {type=Vector}
 R1 Velocity Transform Velocity {type=Vector}
 R2 Velocity Transform Location {type=Vector}
 R2 Velocity Transform Velocity {type=Vector}
 Skel Mesh {type=SkeletalMeshComponent}
 Spring Alpha 02 {type=number}
 Spring Alpha 03 {type=number}
 Spring Alpha 04 {type=number}
 SpringAlpha {type=number}
 Velocity Interp Speed {type=number}
 VestAlpha {type=number}
 WPO Pho Wind Intensity {type=number}
]
